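# Widget Blueprint for a float slider variant picker in Fortnite's customization UI, built on the C++ FortFloatSliderVariantPicker.
# An uppercase variant name header sits above an overlay stacking a SimpleMaterialProgressBar, centered value text and an AnalogSlider.
# The slider's bar and handle are fully transparent, so the progress bar supplies the visible fill.
# OnUpdateValue sets the bar's progress from the normalized value and makes it transparent when locked, white tint otherwise.
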WidgetBlueprint /FortniteCoreUI/Customization/AthenaFloatSliderVariantPicker
graphs [EventGraph {nodes [
   n1 = Event OnUpdateValue!() -> (OutputDelegate, then, NormalizedValue:float, bIsLocked:bool) {is=title K2Node_Event}
   > SetProgress!(
    execute
    self:Fort Simple Material Progress Bar Object Reference=Get Progress() -> (Progress:Simple Material Progress Bar Object Reference)
    Progress:float=@n1.NormalizedValue
   ) -> (then)
   > SetColorAndOpacity(
    self:User Widget Object Reference=Get Progress()
    InColorAndOpacity:Linear Color Structure=SelectColor(A=(R=0,G=0,B=0,A=0), B=(R=1,G=1,B=1,A=1), bPickA=@n1.bIsLocked)
   )
   Event Construct()
   Event Tick()
   Event PreConstruct()
  ]}]
variables []
widgets {
 info {
  inheritedWidgetCount=3
  parentClass=/Script/FortniteUI.FortFloatSliderVariantPicker
  rootWidgetClass=&CommonBorder
  widgetCount=11
 }
 widgets [
  CommonBorder_1 {widgetClassPath=&CommonBorder, widget=CommonBorder_1, properties {padding=<ltrb 8 8 8 8>}}
  VerticalBox_0 {
   widgetClassPath=&VerticalBox
   parent=CommonBorder_1
   widget=VerticalBox_0
   slot {class=BorderSlot, properties {padding=<ltrb 8 8 8 8>}}
  }
  HorizontalBox_1 {
   widgetClassPath=&HorizontalBox
   parent=VerticalBox_0
   widget=HorizontalBox_1
   slot {class=VerticalBoxSlot, properties {padding=<ltrb 0 8 0 0>}}
  }
  Image_3 {
   widgetClassPath=&Image
   parent=HorizontalBox_1
   widget=Image_3
   properties {
    visibility=SelfHitTestInvisible
    brush {
     bIsDynamicallyLoaded=false
     drawAs=Image
     imageType=NoImage
     mirroring=NoMirror
     resourceName=None
     resourceObject=null
     tiling=NoTile
     imageSize=<16 8>
     margin=<ltrb 0 0 0 0>
     outlineSettings {
      bUseBrushTransparency=false
      roundingType=HalfHeightRadius
      width=0
      color {colorUseRule=UseColor_Specified, specifiedColor=<rgba 0 0 0 0>}
      cornerRadii=<0 0 0 0>
     }
     tintColor {colorUseRule=UseColor_Specified, specifiedColor=<rgba 1 1 1 1>}
     uVRegion {bIsValid=false, max=<0 0>, min=<0 0>}
    }
    colorAndOpacity=<rgba 0.1329 0.6038 1 0.3>
   }
   slot {class=HorizontalBoxSlot, properties {verticalAlignment=VAlign_Center}}
  }
  Text_VariantName {
   widgetClassPath=&CommonTextBlock
   bInherited=true
   parent=HorizontalBox_1
   widget=Text_VariantName
   properties {
    text=Text Block
    textTransformPolicy=ToUpper
    colorAndOpacity {colorUseRule=UseColor_Specified, specifiedColor=<rgba 0.1329 0.6038 1 1>}
    font {
     bForceMonospaced=false
     bMaterialIsStencil=false
     fontMaterial=null
     fontObject=&BurbankSmall
     letterSpacing=0
     monospacedWidth=1
     size=15
     skewAmount=0
     typefaceFontName=Bold
     outlineSettings {
      bApplyOutlineToDropShadows=false
      bMiteredCorners=false
      bSeparateFillAlpha=false
      outlineMaterial=null
      outlineSize=0
      outlineColor=<rgba 0 0 0 1>
     }
    }
    shadowOffset=<0 0>
   }
   slot {class=HorizontalBoxSlot, properties {padding=<ltrb 8 0 8 -4>}}
  }
  Image_4 {
   widgetClassPath=&Image
   parent=HorizontalBox_1
   widget=Image_4
   properties {
    visibility=SelfHitTestInvisible
    brush {
     bIsDynamicallyLoaded=false
     drawAs=Image
     imageType=NoImage
     mirroring=NoMirror
     resourceName=None
     resourceObject=null
     tiling=NoTile
     imageSize=<16 8>
     margin=<ltrb 0 0 0 0>
     outlineSettings {
      bUseBrushTransparency=false
      roundingType=HalfHeightRadius
      width=0
      color {colorUseRule=UseColor_Specified, specifiedColor=<rgba 0 0 0 0>}
      cornerRadii=<0 0 0 0>
     }
     tintColor {colorUseRule=UseColor_Specified, specifiedColor=<rgba 1 1 1 1>}
     uVRegion {bIsValid=false, max=<0 0>, min=<0 0>}
    }
    colorAndOpacity=<rgba 0.1329 0.6038 1 0.3>
   }
   slot {class=HorizontalBoxSlot, properties {verticalAlignment=VAlign_Center, size {sizeRule=Fill, value=1}}}
  }
  SizeBox_191 {
   widgetClassPath=&SizeBox
   parent=VerticalBox_0
   widget=SizeBox_191
   properties {bOverride_MinDesiredHeight=true, minDesiredHeight=50}
   slot {class=VerticalBoxSlot, properties {padding=<ltrb 0 20 0 0>}}
  }
  Panel_Value {widgetClassPath=&Overlay, bIsVariable=true, parent=SizeBox_191, widget=Panel_Value, slot {class=SizeBoxSlot}}
  Progress {
   widgetClassPath=/Game/Athena/HUD/SimpleMaterialProgressBar
   bIsVariable=true
   parent=Panel_Value
   widget=Progress
   properties {
    bOverrideDefaultColorA=true
    bOverrideDefaultColorB=true
    bOverrideDefaultColorBackground=true
    visibility=HitTestInvisible
    cachedColorA=<rgba 0.0382 0.4452 0.9047 1>
    cachedColorB=<rgba 0.0382 0.4452 0.9047 1>
    cachedColorBackground=<rgba 0.0033 0.0097 0.1022 1>
   }
   slot {class=OverlaySlot, properties {horizontalAlignment=HAlign_Fill, verticalAlignment=VAlign_Fill}}
  }
  Text_Value {
   widgetClassPath=&CommonTextBlock
   bInherited=true
   bIsVariable=true
   parent=Panel_Value
   widget=Text_Value
   properties {
    justification=Center
    scrollStyle=&TextScrollStyle-Base_C
    style=/FortUILibrary/Styles/Typography/Foundation/TextStyle-Base-S-B.TextStyle-Base-S-B_C
    text="{Value}"
    visibility=HitTestInvisible
    font {
     bForceMonospaced=false
     bMaterialIsStencil=false
     fontMaterial=null
     fontObject=&NotoSans
     letterSpacing=0
     monospacedWidth=1
     size=18
     skewAmount=0
     typefaceFontName=NotoSans-Bold
     outlineSettings {
      bApplyOutlineToDropShadows=false
      bMiteredCorners=false
      bSeparateFillAlpha=false
      outlineMaterial=null
      outlineSize=0
      outlineColor=<rgba 0 0 0 1>
     }
    }
    shadowOffset=<0 0>
   }
   slot {class=OverlaySlot, properties {horizontalAlignment=HAlign_Fill, verticalAlignment=VAlign_Center}}
  }
  Slider_Value {
   widgetClassPath=/Script/CommonUI.AnalogSlider
   bInherited=true
   bIsVariable=true
   parent=Panel_Value
   widget=Slider_Value
   properties {sliderBarColor=<rgba 1 1 1 0>, sliderHandleColor=<rgba 1 1 1 0>}
   slot {class=OverlaySlot, properties {horizontalAlignment=HAlign_Fill, verticalAlignment=VAlign_Fill}}
  }
 ]
}
