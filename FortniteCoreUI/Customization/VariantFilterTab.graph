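# Widget Blueprint for an icon tab that filters customization variants, built on FortVariantFilterTab.
# PreConstruct makes its UIKit enhanced tab background selectable and sets transition timings, then hover, press and selection events drive that block's visual states.
# Selection also plays a sped-up animation fading the icon tint from white to black, reversed on deselection.
# Construct fixes the size box to a square on mobile and clears the overrides otherwise.
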WidgetBlueprint /FortniteCoreUI/Customization/VariantFilterTab
animations [Anim_Focused_Transition {
  binding_count=1
  length_seconds=1
  object=&self:Anim_Focused_Transition
  sequence {bindings [LazyImage_Icon {
     id=#1
     possessed_class=FortLazyImage
     tracks [{
       type=MovieSceneColorTrack
       display_name=Color and Opacity
       sections [{
         type=MovieSceneColorSection
         end_frame=20
         start_frame=0
         channels [
          Color.R {type=MovieSceneScriptingFloatChannel, keys [[0,1],[20,0]]}
          Color.G {type=MovieSceneScriptingFloatChannel, keys [[0,1],[20,0]]}
          Color.B {type=MovieSceneScriptingFloatChannel, keys [[0,1],[20,0]]}
          Color.A {type=MovieSceneScriptingFloatChannel, const [1,1]}
         ]
        }]
      }]
    }]}
 }]
graphs [EventGraph {nodes [
   Event PreConstruct()
   > SetIsSelectable!(
    execute
    self:WBP UIKit Block Base Object Reference=Get BackgroundBlock() -> (BackgroundBlock:WBP UIKit Block Background Tab Enhanced Object Reference)
    Is Selectable:bool=true
   ) -> (then)
   > SetBlockTiming!(
    execute
    self:UIKit Block Object Reference=Get BackgroundBlock()
    Timing:"UIKit Block Timing Structure (by ref)"=Make UIKit Block Timing!(
     Hovering:float=0
     Unhovering:float=0
     Focusing:float=0.3
     Unfocusing:float=0.15
     Pressing:float=0.1
     Releasing:float=0.1
     Disabling:float=0.1
     Enabling:float=0.1
     Selecting:float=0.2
     DeselectingIdle:float=0.2
     DeselectingFocused:float=0.2
    ) -> (UIKitBlockTiming:UIKit Block Timing Structure) {is=title K2Node_MakeStruct}
   ) -> (then)
   Event Construct()
   > n1 = Branch(Condition=IsMobileGame())
   n2 = Get SizeBoxMobile() -> (SizeBoxMobile:Size Box Object Reference)
   > SetWidthOverride(self=@n2, InWidthOverride=80)
   > SetHeightOverride(self=@n2, InHeightOverride=80)
   n3 = Get SizeBoxMobile()
   ClearWidthOverride(execute=@n1.else, self:Size Box Object Reference=@n3)
   > ClearHeightOverride(self:Size Box Object Reference=@n3)
   Event On Hovered()
   > RequestTransition!(execute, self:UIKit Block Object Reference=Get BackgroundBlock(), State:EUIKitBlockVisualState Enum=Focused) -> (then)
   Event On Unhovered()
   > RequestTransition(self=Get BackgroundBlock(), State=Unfocused)
   Event On Selected()
   > SetIsSelected!(execute, self:WBP UIKit Block Base Object Reference=Get BackgroundBlock(), IsSelected:bool=true) -> (then)
   > RequestTransition(
    self=Get BackgroundBlock()
    State=Select(Option 0:EUIKitBlockVisualState Enum=Selected, Option 1:EUIKitBlockVisualState Enum=Focused, Index=IsHovered()) -> (ReturnValue:EUIKitBlockVisualState Enum)
   )
   > PlayAnimation(InAnimation=Get Anim_Focused_Transition() -> (Anim_Focused_Transition:Widget Animation Object Reference), PlaybackSpeed=5)
   Event On Deselected()
   > SetIsSelected(self=Get BackgroundBlock(), IsSelected=false)
   > RequestTransition(self=Get BackgroundBlock(), State=Select(Option 0=DeselectedIdle, Option 1=DeselectedFocused, Index=IsHovered()))
   > PlayAnimation(InAnimation=Get Anim_Focused_Transition(), PlayMode=Reverse, PlaybackSpeed=5)
   Event On Pressed() {is=title K2Node_Event}
   > RequestTransition(self=Get BackgroundBlock(), State=Pressed)
   Event On Released() {is=title K2Node_Event}
   > RequestTransition(self=Get BackgroundBlock(), State=Released)
  ]}]
variables []
widgets {
 info {inheritedWidgetCount=3, parentClass=/Script/FortniteUI.FortVariantFilterTab, rootWidgetClass=&CommonBorder, widgetCount=7}
 widgets [
  CommonBorder_0 {widgetClassPath=&CommonBorder, widget=CommonBorder_0, properties {padding=<ltrb 0 0 0 0>}}
  SizeBoxMobile {
   widgetClassPath=&SizeBox
   bIsVariable=true
   parent=CommonBorder_0
   widget=SizeBoxMobile
   slot {class=BorderSlot, properties {padding=<ltrb 0 0 0 0>}}
  }
  Overlay_0 {widgetClassPath=&Overlay, parent=SizeBoxMobile, widget=Overlay_0, slot {class=SizeBoxSlot}}
  BackgroundBlock {
   widgetClassPath=/FortUILibrary/UIKit/Buttons/Blocks/WBP_UIKit_BlockBackground_Tab_Enhanced
   bIsVariable=true
   parent=Overlay_0
   widget=BackgroundBlock
   properties {outlineMaterial=/FortUILibrary/UIKit/Buttons/MaterialInstances/CTAButtons/MI_UIKit_ButtonTabOutline_Enhanced}
   slot {class=OverlaySlot, properties {horizontalAlignment=HAlign_Fill, verticalAlignment=VAlign_Fill}}
  }
  LazyImage_Icon {
   widgetClassPath=&FortLazyImage
   bInherited=true
   bIsVariable=true
   parent=Overlay_0
   widget=LazyImage_Icon
   slot {class=OverlaySlot, properties {horizontalAlignment=HAlign_Center, verticalAlignment=VAlign_Center}}
  }
  TextBlock_Label {widgetClassPath=&CommonTextBlock, bInherited=true, widget=null}
  InputActionWidget {widgetClassPath=&CommonActionWidget, bInherited=true, widget=null}
 ]
}
