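# Secondary variant filter tab widget for the customization UI, based on the native FortVariantFilterTab class.
# A bordered visibility switcher holds either a lazy-loaded icon inside a size box or a small left-aligned text label.
# On construct, the icon's size box is fixed to a square on mobile and has its size overrides cleared otherwise.
# Clicking the tab plays a generic front-end menu sound.
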WidgetBlueprint /FortniteCoreUI/Customization/VariantFilterTab_Secondary
graphs [EventGraph {nodes [
   Event Construct()
   > n1 = Branch(Condition=IsMobileGame())
   n2 = Get SizeBox_Mobile() -> (SizeBox_Mobile:Size Box Object Reference)
   > SetWidthOverride(self=@n2, InWidthOverride=80)
   > SetHeightOverride(self=@n2, InHeightOverride=80)
   n3 = Get SizeBox_Mobile()
   ClearWidthOverride(execute=@n1.else, self:Size Box Object Reference=@n3)
   > ClearHeightOverride(self:Size Box Object Reference=@n3)
   Event On Clicked()
   > PlaySound2D(Sound=/Game/Sounds/Fort_UI_Sounds/FrontEnd/New_FrontEnd/Fortnite_UI_MainMenu_Generic_02_Cue)
   Event PreConstruct()
   Event Tick()
  ]}]
variables []
widgets {
 info {inheritedWidgetCount=3, parentClass=/Script/FortniteUI.FortVariantFilterTab, rootWidgetClass=&CommonBorder, widgetCount=6}
 widgets [
  CommonBorder_0 {widgetClassPath=&CommonBorder, widget=CommonBorder_0, properties {padding=<ltrb 6 3 6 3>}}
  CommonVisibilitySwitcher_69 {
   widgetClassPath=&CommonVisibilitySwitcher
   parent=CommonBorder_0
   widget=CommonVisibilitySwitcher_69
   slot {class=BorderSlot, properties {padding=<ltrb 6 3 6 3>}}
  }
  SizeBox_Mobile {
   widgetClassPath=&SizeBox
   bIsVariable=true
   parent=CommonVisibilitySwitcher_69
   widget=SizeBox_Mobile
   slot {class=CommonVisibilitySwitcherSlot}
  }
  LazyImage_Icon {
   widgetClassPath=&FortLazyImage
   bInherited=true
   bIsVariable=true
   parent=SizeBox_Mobile
   widget=LazyImage_Icon
   slot {class=SizeBoxSlot}
  }
  TextBlock_Label {
   widgetClassPath=&CommonTextBlock
   bInherited=true
   parent=CommonVisibilitySwitcher_69
   widget=TextBlock_Label
   properties {
    style=/FortUILibrary/Styles/Typography/Foundation/TextStyle-BurbankSmall-XS-White.TextStyle-BurbankSmall-XS-White_C
    text=Text Block
    font {
     bForceMonospaced=false
     bMaterialIsStencil=false
     fontMaterial=null
     fontObject=&BurbankSmall
     letterSpacing=0
     monospacedWidth=1
     size=13
     skewAmount=0
     typefaceFontName=Black
     outlineSettings {
      bApplyOutlineToDropShadows=false
      bMiteredCorners=false
      bSeparateFillAlpha=false
      outlineMaterial=null
      outlineSize=0
      outlineColor=<rgba 0 0 0 1>
     }
    }
    shadowOffset=<0 0>
   }
   slot {class=CommonVisibilitySwitcherSlot, properties {horizontalAlignment=HAlign_Left, verticalAlignment=VAlign_Center}}
  }
  InputActionWidget {widgetClassPath=&CommonActionWidget, bInherited=true, widget=null}
 ]
}
