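# Widget Blueprint that lays out the FortVariantPicker screen for choosing cosmetic variants when customizing a hero's items.
# A header holds primary and secondary filter titles, gamepad tab-cycle prompts, dynamic tab entry boxes and a "no capes" back-bling notice.
# Below it, a styled scroll box holds the variant option entry box, with a collapsed non-scrolling entry box as an alternate.
# The event graph has only empty event stubs, so behavior comes from the native parent through inherited widget bindings.
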WidgetBlueprint /FortniteCoreUI/Customization/AthenaVariantCustomizationSelector
graphs [EventGraph {nodes [
   Event PreConstruct()
   Event Construct()
   Event Tick()
  ]}]
variables []
widgets {
 info {inheritedWidgetCount=15, parentClass=/Script/FortniteUI.FortVariantPicker, rootWidgetClass=&CommonBorder, widgetCount=27}
 widgets [
  CommonBorder_0 {
   widgetClassPath=&CommonBorder
   widget=CommonBorder_0
   properties {
    navigation=&self:WidgetTree.CommonBorder_0.WidgetNavigation_0
    padding=<ltrb 0 0 0 0>
   }
  }
  SizeBoxVariants {
   widgetClassPath=&SizeBox
   parent=CommonBorder_0
   widget=SizeBoxVariants
   properties {bOverride_HeightOverride=true, bOverride_MinDesiredWidth=true, heightOverride=576, minDesiredWidth=872}
   slot {class=BorderSlot, properties {padding=<ltrb 0 0 0 0>}}
  }
  VBox_MainStack {widgetClassPath=&VerticalBox, parent=SizeBoxVariants, widget=VBox_MainStack, slot {class=SizeBoxSlot}}
  SizeBox_0 {
   widgetClassPath=&SizeBox
   parent=VBox_MainStack
   widget=SizeBox_0
   properties {bOverride_MinDesiredHeight=true, minDesiredHeight=74}
   slot {class=VerticalBoxSlot, properties {padding=<ltrb 0 0 0 5>}}
  }
  VerticalBox_1 {widgetClassPath=&VerticalBox, parent=SizeBox_0, widget=VerticalBox_1, slot {class=SizeBoxSlot}}
  Text_NoCapes {
   widgetClassPath=&CommonTextBlock
   bInherited=true
   parent=VerticalBox_1
   widget=Text_NoCapes
   properties {
    autoWrapText=true
    clipping=OnDemand
    mobileFontSizeMultiplier=1.3
    text="Cape Included: Feel free to wear other back-bling, but NO CAPES!"
    colorAndOpacity {colorUseRule=UseColor_Specified, specifiedColor=<rgba 0.1329 0.6038 1 1>}
    font {
     bForceMonospaced=false
     bMaterialIsStencil=false
     fontMaterial=null
     fontObject=&BurbankSmall
     letterSpacing=0
     monospacedWidth=1
     size=18
     skewAmount=0
     typefaceFontName=Bold
     outlineSettings {
      bApplyOutlineToDropShadows=false
      bMiteredCorners=false
      bSeparateFillAlpha=false
      outlineMaterial=null
      outlineSize=0
      outlineColor=<rgba 0 0 0 1>
     }
    }
    margin=<ltrb 0 5 0 10>
    shadowOffset=<0 0>
   }
   slot {class=VerticalBoxSlot, properties {verticalAlignment=VAlign_Center, size {sizeRule=Fill, value=1}}}
  }
  Container_Tabs {
   widgetClassPath=&SizeBox
   bInherited=true
   parent=VerticalBox_1
   widget=Container_Tabs
   properties {bOverride_MinDesiredHeight=true, minDesiredHeight=74}
   slot {class=VerticalBoxSlot}
  }
  Overlay_0 {widgetClassPath=&Overlay, parent=Container_Tabs, widget=Overlay_0, slot {class=SizeBoxSlot}}
  HorizontalBox_0 {
   widgetClassPath=&HorizontalBox
   parent=Overlay_0
   widget=HorizontalBox_0
   properties {flowDirectionPreference=Culture}
   slot {class=OverlaySlot, properties {horizontalAlignment=HAlign_Fill}}
  }
  TabsSizeEnforcerMobile {
   widgetClassPath=&Overlay
   parent=HorizontalBox_0
   widget=TabsSizeEnforcerMobile
   slot {class=HorizontalBoxSlot, properties {horizontalAlignment=HAlign_Left, verticalAlignment=VAlign_Center}}
  }
  maxMobileWidth {
   widgetClassPath=&SizeBox
   parent=TabsSizeEnforcerMobile
   widget=maxMobileWidth
   properties {bOverride_MaxDesiredWidth=true, maxDesiredWidth=600}
   slot {class=OverlaySlot, properties {verticalAlignment=VAlign_Fill}}
  }
  Text_PrimaryFilter {
   widgetClassPath=&CommonTextBlock
   bInherited=true
   bIsVariable=true
   parent=maxMobileWidth
   widget=Text_PrimaryFilter
   properties {
    mobileFontSizeMultiplier=1.3
    scrollStyle=&TextScrollStyle-Base_C
    text=CUSTOMIZE YOUR HERO
    textTransformPolicy=ToUpper
    colorAndOpacity {colorUseRule=UseColor_Specified, specifiedColor=<rgba 0.1329 0.6038 1 1>}
    font {
     bForceMonospaced=false
     bMaterialIsStencil=false
     fontMaterial=null
     fontObject=&Burbank
     letterSpacing=0
     monospacedWidth=1
     size=25
     skewAmount=0
     typefaceFontName=Burbank-Black
     outlineSettings {
      bApplyOutlineToDropShadows=false
      bMiteredCorners=false
      bSeparateFillAlpha=false
      outlineMaterial=null
      outlineSize=0
      outlineColor=<rgba 0 0 0 1>
     }
    }
    renderTransform {angle=0, scale=<1 1>, shear=<-8 0>, translation=<0 0>}
    shadowOffset=<0 0>
   }
   slot {class=SizeBoxSlot, properties {padding=<ltrb 0 8 0 0>}}
  }
  HBox_PrimaryTabNav {
   widgetClassPath=&HorizontalBox
   bInherited=true
   parent=HorizontalBox_0
   widget=HBox_PrimaryTabNav
   slot {
    class=HorizontalBoxSlot
    properties {horizontalAlignment=HAlign_Right, verticalAlignment=VAlign_Center, padding=<ltrb 0 0 8 0>, size {sizeRule=Fill, value=1}}
   }
  }
  ActionWidget_PrimaryTabLeft {
   widgetClassPath=&CommonActionWidget
   bInherited=true
   bIsVariable=true
   parent=HBox_PrimaryTabNav
   widget=ActionWidget_PrimaryTabLeft
   properties {inputActions [{dataTable=&InputActionDataTable, rowName=SecondaryTabLeft}]}
   slot {class=HorizontalBoxSlot, properties {padding=<ltrb 0 0 8 0>}}
  }
  EntryBox_PrimaryTabs {
   widgetClassPath=&DynamicEntryBox
   bInherited=true
   bIsVariable=true
   parent=HBox_PrimaryTabNav
   widget=EntryBox_PrimaryTabs
   properties {entryWidgetClass=&/VariantFilterTab.VariantFilterTab_C, numDesignerPreviewEntries=2}
   slot {class=HorizontalBoxSlot}
  }
  ActionWidget_PrimaryTabRight {
   widgetClassPath=&CommonActionWidget
   bInherited=true
   bIsVariable=true
   parent=HBox_PrimaryTabNav
   widget=ActionWidget_PrimaryTabRight
   properties {inputActions [{dataTable=&InputActionDataTable, rowName=SecondaryTabRight}]}
   slot {class=HorizontalBoxSlot, properties {padding=<ltrb 8 0 0 0>}}
  }
  Container_SecondaryTabs {
   widgetClassPath=&HorizontalBox
   bInherited=true
   parent=VerticalBox_1
   widget=Container_SecondaryTabs
   slot {class=VerticalBoxSlot, properties {verticalAlignment=VAlign_Center, padding=<ltrb 0 8 0 8>}}
  }
  TabsSizeEnforcerMobile_1 {
   widgetClassPath=&Overlay
   parent=Container_SecondaryTabs
   widget=TabsSizeEnforcerMobile_1
   slot {class=HorizontalBoxSlot, properties {horizontalAlignment=HAlign_Left, verticalAlignment=VAlign_Center}}
  }
  maxMobileWidth_1 {
   widgetClassPath=&SizeBox
   parent=TabsSizeEnforcerMobile_1
   widget=maxMobileWidth_1
   properties {bOverride_MaxDesiredWidth=true, maxDesiredWidth=600}
   slot {class=OverlaySlot, properties {verticalAlignment=VAlign_Fill}}
  }
  Text_SecondaryFilter {
   widgetClassPath=&CommonTextBlock
   bInherited=true
   bIsVariable=true
   parent=maxMobileWidth_1
   widget=Text_SecondaryFilter
   properties {
    mobileFontSizeMultiplier=1.4
    scrollStyle=&TextScrollStyle-Base_C
    text=OUTFIT - ALL
    textTransformPolicy=ToUpper
    colorAndOpacity {colorUseRule=UseColor_Specified, specifiedColor=<rgba 0.1329 0.6038 1 1>}
    font {
     bForceMonospaced=false
     bMaterialIsStencil=false
     fontMaterial=null
     fontObject=&BurbankSmall
     letterSpacing=0
     monospacedWidth=1
     size=15
     skewAmount=0
     typefaceFontName=Black
     outlineSettings {
      bApplyOutlineToDropShadows=false
      bMiteredCorners=false
      bSeparateFillAlpha=false
      outlineMaterial=null
      outlineSize=0
      outlineColor=<rgba 0 0 0 1>
     }
    }
    shadowOffset=<0 0>
   }
   slot {class=SizeBoxSlot}
  }
  HorizontalBox_1 {
   widgetClassPath=&HorizontalBox
   parent=Container_SecondaryTabs
   widget=HorizontalBox_1
   slot {
    class=HorizontalBoxSlot
    properties {horizontalAlignment=HAlign_Right, verticalAlignment=VAlign_Center, padding=<ltrb 0 0 8 0>, size {sizeRule=Fill, value=1}}
   }
  }
  ActionWidget_SecondaryTabLeft {
   widgetClassPath=&CommonActionWidget
   bInherited=true
   bIsVariable=true
   parent=HorizontalBox_1
   widget=ActionWidget_SecondaryTabLeft
   properties {inputActions [{dataTable=&InputActionDataTable, rowName=SecondaryTabLeft}]}
   slot {class=HorizontalBoxSlot, properties {horizontalAlignment=HAlign_Left, padding=<ltrb 0 0 8 0>}}
  }
  EntryBox_SecondaryTabs {
   widgetClassPath=&DynamicEntryBox
   bInherited=true
   bIsVariable=true
   parent=HorizontalBox_1
   widget=EntryBox_SecondaryTabs
   properties {entryWidgetClass=&/VariantFilterTab_Secondary.VariantFilterTab_Secondary_C, numDesignerPreviewEntries=2}
   slot {class=HorizontalBoxSlot, properties {verticalAlignment=VAlign_Center}}
  }
  ActionWidget_SecondaryTabRight {
   widgetClassPath=&CommonActionWidget
   bInherited=true
   bIsVariable=true
   parent=HorizontalBox_1
   widget=ActionWidget_SecondaryTabRight
   properties {inputActions [{dataTable=&InputActionDataTable, rowName=SecondaryTabRight}]}
   slot {class=HorizontalBoxSlot, properties {padding=<ltrb 8 0 0 0>}}
  }
  ScrollBox_Variants {
   widgetClassPath=&ScrollBox
   bInherited=true
   parent=VBox_MainStack
   widget=ScrollBox_Variants
   properties {
    scrollWhenFocusChanges=AnimatedScroll
    widgetBarStyle {
     thickness=16
     draggedThumbImage {
      bIsDynamicallyLoaded=false
      drawAs=Image
      imageType=FullColor
      mirroring=NoMirror
      resourceName=../../../Engine/Content/Slate/Common/Scrollbar_Thumb.png
      resourceObject=/FortUILibrary/Foundation/Art_ToolKit/AngledBoxes/M_UI_ItemShop_ScrollBar_Drag
      tiling=NoTile
      imageSize=<0 0>
      margin=<ltrb 0.25 0.25 0.25 0.25>
      outlineSettings {
       bUseBrushTransparency=false
       roundingType=HalfHeightRadius
       width=0
       color {colorUseRule=UseColor_Specified, specifiedColor=<rgba 0 0 0 0>}
       cornerRadii=<0 0 0 0>
      }
      tintColor {colorUseRule=UseColor_Specified, specifiedColor=<rgba 1 1 1 1>}
      uVRegion {bIsValid=false, max=<0 0>, min=<0 0>}
     }
     horizontalBackgroundImage {
      bIsDynamicallyLoaded=false
      drawAs=NoDrawType
      imageType=NoImage
      mirroring=NoMirror
      resourceName=None
      resourceObject=null
      tiling=NoTile
      imageSize=<0 0>
      margin=<ltrb 0 0 0 0>
      outlineSettings {
       bUseBrushTransparency=false
       roundingType=HalfHeightRadius
       width=0
       color {colorUseRule=UseColor_Specified, specifiedColor=<rgba 0 0 0 0>}
       cornerRadii=<0 0 0 0>
      }
      tintColor {colorUseRule=UseColor_Specified, specifiedColor=<rgba 1 1 1 1>}
      uVRegion {bIsValid=false, max=<0 0>, min=<0 0>}
     }
     horizontalBottomSlotImage {
      bIsDynamicallyLoaded=false
      drawAs=NoDrawType
      imageType=FullColor
      mirroring=NoMirror
      resourceName=../../../Engine/Content/Slate/Common/Scrollbar_Background_Horizontal.png
      resourceObject=null
      tiling=NoTile
      imageSize=<8 8>
      margin=<ltrb 0 0 0 0>
      outlineSettings {
       bUseBrushTransparency=false
       roundingType=HalfHeightRadius
       width=0
       color {colorUseRule=UseColor_Specified, specifiedColor=<rgba 0 0 0 0>}
       cornerRadii=<0 0 0 0>
      }
      tintColor {colorUseRule=UseColor_Specified, specifiedColor=<rgba 1 1 1 1>}
      uVRegion {bIsValid=false, max=<0 0>, min=<0 0>}
     }
     horizontalTopSlotImage {
      bIsDynamicallyLoaded=false
      drawAs=NoDrawType
      imageType=FullColor
      mirroring=NoMirror
      resourceName=../../../Engine/Content/Slate/Common/Scrollbar_Background_Horizontal.png
      resourceObject=null
      tiling=NoTile
      imageSize=<8 8>
      margin=<ltrb 0 0 0 0>
      outlineSettings {
       bUseBrushTransparency=false
       roundingType=HalfHeightRadius
       width=0
       color {colorUseRule=UseColor_Specified, specifiedColor=<rgba 0 0 0 0>}
       cornerRadii=<0 0 0 0>
      }
      tintColor {colorUseRule=UseColor_Specified, specifiedColor=<rgba 1 1 1 1>}
      uVRegion {bIsValid=false, max=<0 0>, min=<0 0>}
     }
     hoveredThumbImage {
      bIsDynamicallyLoaded=false
      drawAs=Image
      imageType=FullColor
      mirroring=NoMirror
      resourceName=../../../Engine/Content/Slate/Common/Scrollbar_Thumb.png
      resourceObject=/FortUILibrary/Foundation/Art_ToolKit/AngledBoxes/M_UI_ItemShop_ScrollBar_Hover
      tiling=NoTile
      imageSize=<0 0>
      margin=<ltrb 0.25 0.25 0.25 0.25>
      outlineSettings {
       bUseBrushTransparency=false
       roundingType=HalfHeightRadius
       width=0
       color {colorUseRule=UseColor_Specified, specifiedColor=<rgba 0 0 0 0>}
       cornerRadii=<0 0 0 0>
      }
      tintColor {colorUseRule=UseColor_Specified, specifiedColor=<rgba 1 1 1 1>}
      uVRegion {bIsValid=false, max=<0 0>, min=<0 0>}
     }
     normalThumbImage {
      bIsDynamicallyLoaded=false
      drawAs=Image
      imageType=FullColor
      mirroring=NoMirror
      resourceName=../../../Engine/Content/Slate/Common/Scrollbar_Thumb.png
      resourceObject=/FortUILibrary/Foundation/Art_ToolKit/AngledBoxes/M_UI_Locker_ScrollBar_Base
      tiling=NoTile
      imageSize=<0 0>
      margin=<ltrb 0.25 0.25 0.25 0.25>
      outlineSettings {
       bUseBrushTransparency=false
       roundingType=HalfHeightRadius
       width=0
       color {colorUseRule=UseColor_Specified, specifiedColor=<rgba 0 0 0 0>}
       cornerRadii=<0 0 0 0>
      }
      tintColor {colorUseRule=UseColor_Specified, specifiedColor=<rgba 1 1 1 1>}
      uVRegion {bIsValid=false, max=<0 0>, min=<0 0>}
     }
     verticalBackgroundImage {
      bIsDynamicallyLoaded=false
      drawAs=Image
      imageType=NoImage
      mirroring=NoMirror
      resourceName=None
      resourceObject=/FortUILibrary/Foundation/Art_ToolKit/AngledBoxes/M_UI_ItemShop_ScrollBar_Backing
      tiling=NoTile
      imageSize=<0 0>
      margin=<ltrb 0 0 0 0>
      outlineSettings {
       bUseBrushTransparency=false
       roundingType=HalfHeightRadius
       width=0
       color {colorUseRule=UseColor_Specified, specifiedColor=<rgba 0 0 0 0>}
       cornerRadii=<0 0 0 0>
      }
      tintColor {colorUseRule=UseColor_Specified, specifiedColor=<rgba 1 1 1 1>}
      uVRegion {bIsValid=false, max=<0 0>, min=<0 0>}
     }
     verticalBottomSlotImage {
      bIsDynamicallyLoaded=false
      drawAs=NoDrawType
      imageType=FullColor
      mirroring=NoMirror
      resourceName=../../../Engine/Content/Slate/Common/Scrollbar_Background_Vertical.png
      resourceObject=null
      tiling=NoTile
      imageSize=<8 8>
      margin=<ltrb 0 0 0 0>
      outlineSettings {
       bUseBrushTransparency=false
       roundingType=HalfHeightRadius
       width=0
       color {colorUseRule=UseColor_Specified, specifiedColor=<rgba 0 0 0 0>}
       cornerRadii=<0 0 0 0>
      }
      tintColor {colorUseRule=UseColor_Specified, specifiedColor=<rgba 1 1 1 1>}
      uVRegion {bIsValid=false, max=<0 0>, min=<0 0>}
     }
     verticalTopSlotImage {
      bIsDynamicallyLoaded=false
      drawAs=NoDrawType
      imageType=FullColor
      mirroring=NoMirror
      resourceName=../../../Engine/Content/Slate/Common/Scrollbar_Background_Vertical.png
      resourceObject=null
      tiling=NoTile
      imageSize=<0 0>
      margin=<ltrb 0 0 0 0>
      outlineSettings {
       bUseBrushTransparency=false
       roundingType=HalfHeightRadius
       width=0
       color {colorUseRule=UseColor_Specified, specifiedColor=<rgba 0 0 0 0>}
       cornerRadii=<0 0 0 0>
      }
      tintColor {colorUseRule=UseColor_Specified, specifiedColor=<rgba 1 1 1 1>}
      uVRegion {bIsValid=false, max=<0 0>, min=<0 0>}
     }
    }
    widgetStyle {
     barThickness=16
     bottomShadowBrush {
      bIsDynamicallyLoaded=false
      drawAs=Box
      imageType=FullColor
      mirroring=NoMirror
      resourceName=../../../Engine/Content/Slate/Common/ScrollBoxShadowBottom.png
      resourceObject=null
      tiling=NoTile
      imageSize=<16 8>
      margin=<ltrb 0.5 0 0.5 1>
      outlineSettings {
       bUseBrushTransparency=false
       roundingType=HalfHeightRadius
       width=0
       color {colorUseRule=UseColor_Specified, specifiedColor=<rgba 0 0 0 0>}
       cornerRadii=<0 0 0 0>
      }
      tintColor {colorUseRule=UseColor_Specified, specifiedColor=<rgba 1 1 1 1>}
      uVRegion {bIsValid=false, max=<0 0>, min=<0 0>}
     }
     horizontalScrolledContentPadding=<ltrb 0 0 1 0>
     leftShadowBrush {
      bIsDynamicallyLoaded=false
      drawAs=Box
      imageType=FullColor
      mirroring=NoMirror
      resourceName=../../../Engine/Content/Slate/Common/ScrollBoxShadowLeft.png
      resourceObject=null
      tiling=NoTile
      imageSize=<8 16>
      margin=<ltrb 1 0.5 0 0.5>
      outlineSettings {
       bUseBrushTransparency=false
       roundingType=HalfHeightRadius
       width=0
       color {colorUseRule=UseColor_Specified, specifiedColor=<rgba 0 0 0 0>}
       cornerRadii=<0 0 0 0>
      }
      tintColor {colorUseRule=UseColor_Specified, specifiedColor=<rgba 1 1 1 1>}
      uVRegion {bIsValid=false, max=<0 0>, min=<0 0>}
     }
     rightShadowBrush {
      bIsDynamicallyLoaded=false
      drawAs=Box
      imageType=FullColor
      mirroring=NoMirror
      resourceName=../../../Engine/Content/Slate/Common/ScrollBoxShadowRight.png
      resourceObject=null
      tiling=NoTile
      imageSize=<8 16>
      margin=<ltrb 0 0.5 1 0.5>
      outlineSettings {
       bUseBrushTransparency=false
       roundingType=HalfHeightRadius
       width=0
       color {colorUseRule=UseColor_Specified, specifiedColor=<rgba 0 0 0 0>}
       cornerRadii=<0 0 0 0>
      }
      tintColor {colorUseRule=UseColor_Specified, specifiedColor=<rgba 1 1 1 1>}
      uVRegion {bIsValid=false, max=<0 0>, min=<0 0>}
     }
     topShadowBrush {
      bIsDynamicallyLoaded=false
      drawAs=Box
      imageType=FullColor
      mirroring=NoMirror
      resourceName=../../../Engine/Content/Slate/Common/ScrollBoxShadowTop.png
      resourceObject=null
      tiling=NoTile
      imageSize=<16 8>
      margin=<ltrb 0.5 1 0.5 0>
      outlineSettings {
       bUseBrushTransparency=false
       roundingType=HalfHeightRadius
       width=0
       color {colorUseRule=UseColor_Specified, specifiedColor=<rgba 0 0 0 0>}
       cornerRadii=<0 0 0 0>
      }
      tintColor {colorUseRule=UseColor_Specified, specifiedColor=<rgba 1 1 1 1>}
      uVRegion {bIsValid=false, max=<0 0>, min=<0 0>}
     }
     verticalScrolledContentPadding=<ltrb 0 0 0 1>
    }
   }
   slot {class=VerticalBoxSlot, properties {size {sizeRule=Fill, value=1}}}
  }
  EntryBox_VariantOptions {
   widgetClassPath=/Script/FortniteUI.FortVariantEntryBox
   bInherited=true
   bIsVariable=true
   parent=ScrollBox_Variants
   widget=EntryBox_VariantOptions
   properties {
    entryBoxType=Vertical
    navigation=&self:WidgetTree.EntryBox_VariantOptions.WidgetNavigation_0
    numDesignerPreviewEntries=4
   }
   slot {class=ScrollBoxSlot}
  }
  EntryBox_VariantOptions_NoScroll {
   widgetClassPath=&FortVariantEntryBox
   bInherited=true
   bIsVariable=true
   parent=VBox_MainStack
   widget=EntryBox_VariantOptions_NoScroll
   properties {
    entryBoxType=Vertical
    navigation=&self:WidgetTree.EntryBox_VariantOptions_NoScroll.WidgetNavigation_0
    numDesignerPreviewEntries=4
    visibility=Collapsed
    entrySizeRule {sizeRule=Fill, value=1}
   }
   slot {class=VerticalBoxSlot, properties {size {sizeRule=Fill, value=1}}}
  }
 ]
}
