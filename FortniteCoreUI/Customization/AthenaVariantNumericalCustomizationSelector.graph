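# Widget Blueprint deriving from FortVariantNumericalPicker that lays out the numeric variant picker for Athena cosmetic customization.
# An uppercase variant-name header sits above a bordered value readout, flanked by single- and double-arrow PanelButtons for stepping down and up by ones and tens.
# The text blocks and buttons fill inherited widget slots; the event graph has only unwired event nodes, so behavior is left to the parent class.
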WidgetBlueprint /FortniteCoreUI/Customization/AthenaVariantNumericalCustomizationSelector
graphs [EventGraph {nodes [
   Event PreConstruct()
   Event Construct()
   Event Tick()
  ]}]
variables []
widgets {
 info {
  inheritedWidgetCount=6
  parentClass=/Script/FortniteUI.FortVariantNumericalPicker
  rootWidgetClass=&CommonBorder
  widgetCount=17
 }
 widgets [
  CommonBorder_1 {widgetClassPath=&CommonBorder, widget=CommonBorder_1, properties {padding=<ltrb 8 8 8 8>}}
  VerticalBox_0 {
   widgetClassPath=&VerticalBox
   parent=CommonBorder_1
   widget=VerticalBox_0
   slot {class=BorderSlot, properties {padding=<ltrb 8 8 8 8>}}
  }
  HorizontalBox_1 {
   widgetClassPath=&HorizontalBox
   parent=VerticalBox_0
   widget=HorizontalBox_1
   slot {class=VerticalBoxSlot, properties {padding=<ltrb 0 8 0 16>}}
  }
  Image_3 {
   widgetClassPath=&Image
   parent=HorizontalBox_1
   widget=Image_3
   properties {
    visibility=SelfHitTestInvisible
    brush {
     bIsDynamicallyLoaded=false
     drawAs=Image
     imageType=NoImage
     mirroring=NoMirror
     resourceName=None
     resourceObject=null
     tiling=NoTile
     imageSize=<16 8>
     margin=<ltrb 0 0 0 0>
     outlineSettings {
      bUseBrushTransparency=false
      roundingType=HalfHeightRadius
      width=0
      color {colorUseRule=UseColor_Specified, specifiedColor=<rgba 0 0 0 0>}
      cornerRadii=<0 0 0 0>
     }
     tintColor {colorUseRule=UseColor_Specified, specifiedColor=<rgba 1 1 1 1>}
     uVRegion {bIsValid=false, max=<0 0>, min=<0 0>}
    }
    colorAndOpacity=<rgba 0.1329 0.6038 1 0.3>
   }
   slot {class=HorizontalBoxSlot, properties {verticalAlignment=VAlign_Center}}
  }
  Text_VariantName {
   widgetClassPath=&CommonTextBlock
   bInherited=true
   parent=HorizontalBox_1
   widget=Text_VariantName
   properties {
    text=Text Block
    textTransformPolicy=ToUpper
    colorAndOpacity {colorUseRule=UseColor_Specified, specifiedColor=<rgba 0.1329 0.6038 1 1>}
    font {
     bForceMonospaced=false
     bMaterialIsStencil=false
     fontMaterial=null
     fontObject=&BurbankSmall
     letterSpacing=0
     monospacedWidth=1
     size=15
     skewAmount=0
     typefaceFontName=Bold
     outlineSettings {
      bApplyOutlineToDropShadows=false
      bMiteredCorners=false
      bSeparateFillAlpha=false
      outlineMaterial=null
      outlineSize=0
      outlineColor=<rgba 0 0 0 1>
     }
    }
    shadowOffset=<0 0>
   }
   slot {class=HorizontalBoxSlot, properties {padding=<ltrb 8 0 8 -4>}}
  }
  Image_4 {
   widgetClassPath=&Image
   parent=HorizontalBox_1
   widget=Image_4
   properties {
    visibility=SelfHitTestInvisible
    brush {
     bIsDynamicallyLoaded=false
     drawAs=Image
     imageType=NoImage
     mirroring=NoMirror
     resourceName=None
     resourceObject=null
     tiling=NoTile
     imageSize=<16 8>
     margin=<ltrb 0 0 0 0>
     outlineSettings {
      bUseBrushTransparency=false
      roundingType=HalfHeightRadius
      width=0
      color {colorUseRule=UseColor_Specified, specifiedColor=<rgba 0 0 0 0>}
      cornerRadii=<0 0 0 0>
     }
     tintColor {colorUseRule=UseColor_Specified, specifiedColor=<rgba 1 1 1 1>}
     uVRegion {bIsValid=false, max=<0 0>, min=<0 0>}
    }
    colorAndOpacity=<rgba 0.1329 0.6038 1 0.3>
   }
   slot {class=HorizontalBoxSlot, properties {verticalAlignment=VAlign_Center, size {sizeRule=Fill, value=1}}}
  }
  HorizontalBox_0 {widgetClassPath=&HorizontalBox, parent=VerticalBox_0, widget=HorizontalBox_0, slot {class=VerticalBoxSlot}}
  Button_TenDigitDown {
   widgetClassPath=/FortUILibrary/Foundation/Buttons/PanelButton
   bInherited=true
   bIsVariable=true
   parent=HorizontalBox_0
   widget=Button_TenDigitDown
   properties {
    inputActionHorizontalAlignment=HAlign_Right
    inputActionVerticalAlignment=VAlign_Center
    minHeight=88
    minWidth=88
    onButtonBaseClicked=(&self:WidgetTree.Button_TenDigitDown.HandleOnButtonClicked)
    onButtonBaseDoubleClicked=(&self:WidgetTree.Button_TenDigitDown.HandleOnButtonDoubleClicked)
    onButtonBaseHovered=(&self:WidgetTree.Button_TenDigitDown.HandleOnButtonHovered)
    onButtonBaseUnhovered=(&self:WidgetTree.Button_TenDigitDown.HandleOnButtonUnhovered)
    onSelectedChangedBase=(&self:WidgetTree.Button_TenDigitDown.HandleOnSelectedChanged)
    style=/FortUILibrary/Foundation/Buttons/ButtonStyles/ButtonStyle_CycleArrow_Left_Double.ButtonStyle_CycleArrow_Left_Double_C
    inputActionRenderTranslation=<22 0>
    inputActionRimBrushSize=<40 40>
   }
   slot {class=HorizontalBoxSlot, properties {verticalAlignment=VAlign_Center, padding=<ltrb 0 0 16 0>}}
  }
  Button_ZeroDigitDown {
   widgetClassPath=&PanelButton
   bInherited=true
   bIsVariable=true
   parent=HorizontalBox_0
   widget=Button_ZeroDigitDown
   properties {
    inputActionHorizontalAlignment=HAlign_Right
    inputActionVerticalAlignment=VAlign_Center
    minHeight=88
    minWidth=88
    onButtonBaseClicked=(&self:WidgetTree.Button_ZeroDigitDown.HandleOnButtonClicked)
    onButtonBaseDoubleClicked=(&self:WidgetTree.Button_ZeroDigitDown.HandleOnButtonDoubleClicked)
    onButtonBaseHovered=(&self:WidgetTree.Button_ZeroDigitDown.HandleOnButtonHovered)
    onButtonBaseUnhovered=(&self:WidgetTree.Button_ZeroDigitDown.HandleOnButtonUnhovered)
    onSelectedChangedBase=(&self:WidgetTree.Button_ZeroDigitDown.HandleOnSelectedChanged)
    style=/FortUILibrary/Foundation/Buttons/ButtonStyles/ButtonStyle_CycleArrow_Left.ButtonStyle_CycleArrow_Left_C
    inputActionRenderTranslation=<22 0>
    inputActionRimBrushSize=<40 40>
   }
   slot {class=HorizontalBoxSlot, properties {verticalAlignment=VAlign_Center}}
  }
  SelectionTextSB {
   widgetClassPath=&SizeBox
   parent=HorizontalBox_0
   widget=SelectionTextSB
   properties {bOverride_MinDesiredWidth=true, minDesiredWidth=192}
   slot {class=HorizontalBoxSlot}
  }
  CommonBorder_0 {
   widgetClassPath=&CommonBorder
   parent=SelectionTextSB
   widget=CommonBorder_0
   properties {style=/FortUILibrary/Styles/Border/Border_Solid_DkBlue.Border_Solid_DkBlue_C, brushColor=<rgba 1 1 1 0.2>, padding=<ltrb 0 0 0 0>}
   slot {class=SizeBoxSlot}
  }
  VerticalBox_1 {
   widgetClassPath=&VerticalBox
   parent=CommonBorder_0
   widget=VerticalBox_1
   slot {class=BorderSlot, properties {padding=<ltrb 0 0 0 0>}}
  }
  Image_1 {
   widgetClassPath=&Image
   parent=VerticalBox_1
   widget=Image_1
   properties {
    visibility=SelfHitTestInvisible
    brush {
     bIsDynamicallyLoaded=false
     drawAs=Image
     imageType=NoImage
     mirroring=NoMirror
     resourceName=None
     resourceObject=null
     tiling=NoTile
     imageSize=<16 4>
     margin=<ltrb 0 0 0 0>
     outlineSettings {
      bUseBrushTransparency=false
      roundingType=HalfHeightRadius
      width=0
      color {colorUseRule=UseColor_Specified, specifiedColor=<rgba 0 0 0 0>}
      cornerRadii=<0 0 0 0>
     }
     tintColor {colorUseRule=UseColor_Specified, specifiedColor=<rgba 1 1 1 1>}
     uVRegion {bIsValid=false, max=<0 0>, min=<0 0>}
    }
    colorAndOpacity=<rgba 1 1 1 0.2>
   }
   slot {class=VerticalBoxSlot}
  }
  Text_NumericalValue {
   widgetClassPath=&CommonTextBlock
   bInherited=true
   parent=VerticalBox_1
   widget=Text_NumericalValue
   properties {
    justification=Center
    scrollStyle=&TextScrollStyle-Base_C
    style=/FortUILibrary/Styles/Typography/Foundation/TextStyle-Base-XL-70pc.TextStyle-Base-XL-70pc_C
    text=999
    colorAndOpacity {colorUseRule=UseColor_Specified, specifiedColor=<rgba 1 1 1 0.7>}
    font {
     bForceMonospaced=false
     bMaterialIsStencil=false
     fontMaterial=null
     fontObject=&NotoSans
     letterSpacing=0
     monospacedWidth=1
     size=44
     skewAmount=0
     typefaceFontName=NotoSans-Italic
     outlineSettings {
      bApplyOutlineToDropShadows=false
      bMiteredCorners=false
      bSeparateFillAlpha=false
      outlineMaterial=null
      outlineSize=0
      outlineColor=<rgba 0 0 0 1>
     }
    }
    shadowOffset=<0 0>
   }
   slot {
    class=VerticalBoxSlot
    properties {horizontalAlignment=HAlign_Center, verticalAlignment=VAlign_Center, padding=<ltrb 16 -16 16 -16>, size {sizeRule=Fill, value=1}}
   }
  }
  Image_0 {
   widgetClassPath=&Image
   parent=VerticalBox_1
   widget=Image_0
   properties {
    visibility=SelfHitTestInvisible
    brush {
     bIsDynamicallyLoaded=false
     drawAs=Image
     imageType=NoImage
     mirroring=NoMirror
     resourceName=None
     resourceObject=null
     tiling=NoTile
     imageSize=<16 4>
     margin=<ltrb 0 0 0 0>
     outlineSettings {
      bUseBrushTransparency=false
      roundingType=HalfHeightRadius
      width=0
      color {colorUseRule=UseColor_Specified, specifiedColor=<rgba 0 0 0 0>}
      cornerRadii=<0 0 0 0>
     }
     tintColor {colorUseRule=UseColor_Specified, specifiedColor=<rgba 1 1 1 1>}
     uVRegion {bIsValid=false, max=<0 0>, min=<0 0>}
    }
    colorAndOpacity=<rgba 1 1 1 0.2>
   }
   slot {class=VerticalBoxSlot}
  }
  Button_ZeroDigitUp {
   widgetClassPath=&PanelButton
   bInherited=true
   bIsVariable=true
   parent=HorizontalBox_0
   widget=Button_ZeroDigitUp
   properties {
    inputActionHorizontalAlignment=HAlign_Left
    inputActionVerticalAlignment=VAlign_Center
    minHeight=88
    minWidth=88
    onButtonBaseClicked=(&self:WidgetTree.Button_ZeroDigitUp.HandleOnButtonClicked)
    onButtonBaseDoubleClicked=(&self:WidgetTree.Button_ZeroDigitUp.HandleOnButtonDoubleClicked)
    onButtonBaseHovered=(&self:WidgetTree.Button_ZeroDigitUp.HandleOnButtonHovered)
    onButtonBaseUnhovered=(&self:WidgetTree.Button_ZeroDigitUp.HandleOnButtonUnhovered)
    onSelectedChangedBase=(&self:WidgetTree.Button_ZeroDigitUp.HandleOnSelectedChanged)
    style=/FortUILibrary/Foundation/Buttons/ButtonStyles/ButtonStyle_CycleArrow_Right.ButtonStyle_CycleArrow_Right_C
    inputActionRenderTranslation=<-22 0>
    inputActionRimBrushSize=<40 40>
   }
   slot {class=HorizontalBoxSlot, properties {verticalAlignment=VAlign_Center}}
  }
  Button_TenDigitUp {
   widgetClassPath=&PanelButton
   bInherited=true
   bIsVariable=true
   parent=HorizontalBox_0
   widget=Button_TenDigitUp
   properties {
    inputActionHorizontalAlignment=HAlign_Left
    inputActionVerticalAlignment=VAlign_Center
    minHeight=88
    minWidth=88
    onButtonBaseClicked=(&self:WidgetTree.Button_TenDigitUp.HandleOnButtonClicked)
    onButtonBaseDoubleClicked=(&self:WidgetTree.Button_TenDigitUp.HandleOnButtonDoubleClicked)
    onButtonBaseHovered=(&self:WidgetTree.Button_TenDigitUp.HandleOnButtonHovered)
    onButtonBaseUnhovered=(&self:WidgetTree.Button_TenDigitUp.HandleOnButtonUnhovered)
    onSelectedChangedBase=(&self:WidgetTree.Button_TenDigitUp.HandleOnSelectedChanged)
    style=/FortUILibrary/Foundation/Buttons/ButtonStyles/ButtonStyle_CycleArrow_Right_Double.ButtonStyle_CycleArrow_Right_Double_C
    inputActionRenderTranslation=<-22 0>
    inputActionRimBrushSize=<40 40>
   }
   slot {class=HorizontalBoxSlot, properties {verticalAlignment=VAlign_Center, padding=<ltrb 16 0 0 0>}}
  }
 ]
}
