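# Tile button widget, derived from FortVariantTileButton, that displays a selectable Athena cosmetic variant with locked, active and conflict indicators.
# An overlay stacks a background, lazy-loaded variant image, state icons and a UIKit block outline.
# Construct sizes the image for mobile or other platforms and rounds the outline, hover toggles its focus state, and click plays a sound.
# UpdateConflictStatus rebroadcasts the conflict flag and reason through the OnConflictStatusUpdated dispatcher.
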WidgetBlueprint /FortniteCoreUI/Customization/AthenaVariantTileButton
animations [test {
  binding_count=1
  length_seconds=5
  object=&self:test
  sequence {bindings [Image_VariantDisplay {id=#1, possessed_class=FortLazyImage, tracks [{type=MovieSceneWidgetMaterialTrack, sections []}]}]}
 }]
graphs [
 EventGraph {nodes [
   n1 = GetDynamicMaterial()
   > SetScalarParameterValue(self=@n1, ParameterName=bIsVariant, Value=1)
   Event On Clicked()
   > PlaySound2D(Sound=/Game/Sounds/Fort_UI_Sounds/FrontEnd/ItemShop/Fort_UI_Shop_Menu_Click_Simple_B_Cue)
   n2 = Event UpdateConflictStatus!() -> (OutputDelegate, then, bIsConflicting:bool, ConflictReason:"Text (by ref)") {is=title K2Node_Event}
   > Call On Conflict Status Updated!(
    execute:Exec
    self:Athena Variant Tile Button Object Reference
    IsConflicting:bool=@n2.bIsConflicting
    ConflictReason:Text=@n2.ConflictReason
   ) -> (then:Exec) {is=title K2Node_CallDelegate}
   Event On Hovered()
   > RequestTransition!(
    execute
    self:UIKit Block Object Reference=Get Block_Outline() -> (Block_Outline:WBP UIKit Block Outline Object Reference)
    State:EUIKitBlockVisualState Enum=Focused
   ) -> (then)
   Event On Unhovered()
   > RequestTransition(self=Get Block_Outline(), State=Unfocused)
   Event Construct()
   > SetDesiredSizeOverride!(
    execute
    self:Image Object Reference=Get Image_VariantDisplay() -> (Image_VariantDisplay:Fort Lazy Image Object Reference)
    DesiredSize:Vector 2D Structure=Select(Option 0:Vector 2D Structure=(X=88,Y=88), Option 1:Vector 2D Structure=(X=120,Y=120), Index=IsMobileGame()) -> (ReturnValue:Vector 2D Structure)
   ) -> (then)
   > SetCornerRadiuses!(
    execute
    self:WBP UIKit Block Outline Object Reference=Get Block_Outline()
    TopLeft:float=30
    TopRight:float=30
    BottomLeft:float=30
    BottomRight:float=30
    RadiusInPixels:bool=true
   ) -> (then)
   Event OnListItemObjectSet()
  ]}
 OnConflictStatusUpdated {nodes [OnConflictStatusUpdated!() -> (then, IsConflicting:bool=~, ConflictReason:Text) {is=title K2Node_FunctionEntry}]}
]
variables [OnConflictStatusUpdated {}]
widgets {
 info {inheritedWidgetCount=5, parentClass=/Script/FortniteUI.FortVariantTileButton, rootWidgetClass=&Overlay, widgetCount=9}
 widgets [
  Overlay_0 {widgetClassPath=&Overlay, widget=Overlay_0}
  ImageBackground {
   widgetClassPath=&Image
   bIsVariable=true
   parent=Overlay_0
   widget=ImageBackground
   properties {
    visibility=SelfHitTestInvisible
    brush {
     bIsDynamicallyLoaded=false
     drawAs=Box
     imageType=NoImage
     mirroring=NoMirror
     resourceName=None
     resourceObject=/FortUILibrary/Systems/Items/MaterialInstances/MI_UIKit_Item_Background
     tiling=NoTile
     imageSize=<88 88>
     margin=<ltrb 0.5 0.5 0.5 0.5>
     outlineSettings {
      bUseBrushTransparency=false
      roundingType=HalfHeightRadius
      width=0
      color {colorUseRule=UseColor_Specified, specifiedColor=<rgba 0 0 0 0>}
      cornerRadii=<0 0 0 0>
     }
     tintColor {colorUseRule=UseColor_Specified, specifiedColor=<rgba 1 1 1 1>}
     uVRegion {bIsValid=false, max=<0 0>, min=<0 0>}
    }
   }
   slot {class=OverlaySlot, properties {horizontalAlignment=HAlign_Fill, verticalAlignment=VAlign_Fill}}
  }
  Image_VariantDisplay {
   widgetClassPath=&FortLazyImage
   bInherited=true
   bIsVariable=true
   parent=Overlay_0
   widget=Image_VariantDisplay
   properties {
    brush {
     bIsDynamicallyLoaded=false
     drawAs=Image
     imageType=FullColor
     mirroring=NoMirror
     resourceName=None
     resourceObject=/FortUILibrary/Systems/Items/MaterialInstances/MI_UIKit_ItemLocker
     tiling=NoTile
     imageSize=<88 88>
     margin=<ltrb 0 0 0 0>
     outlineSettings {
      bUseBrushTransparency=false
      roundingType=HalfHeightRadius
      width=0
      color {colorUseRule=UseColor_Specified, specifiedColor=<rgba 0 0 0 0>}
      cornerRadii=<0 0 0 0>
     }
     tintColor {colorUseRule=UseColor_Specified, specifiedColor=<rgba 1 1 1 1>}
     uVRegion {bIsValid=false, max=<0 0>, min=<0 0>}
    }
    loadingBackgroundBrush {
     bIsDynamicallyLoaded=false
     drawAs=Image
     imageType=FullColor
     mirroring=NoMirror
     resourceName=None
     resourceObject=/FortUILibrary/Systems/Items/MaterialInstances/MI_UIKit_Item_Loading
     tiling=NoTile
     imageSize=<32 32>
     margin=<ltrb 0 0 0 0>
     outlineSettings {
      bUseBrushTransparency=false
      roundingType=HalfHeightRadius
      width=0
      color {colorUseRule=UseColor_Specified, specifiedColor=<rgba 0 0 0 0>}
      cornerRadii=<0 0 0 0>
     }
     tintColor {colorUseRule=UseColor_Specified, specifiedColor=<rgba 1 1 1 1>}
     uVRegion {bIsValid=false, max=<0 0>, min=<0 0>}
    }
    loadingThrobberBrush {
     bIsDynamicallyLoaded=false
     drawAs=Image
     imageType=FullColor
     mirroring=NoMirror
     resourceName=None
     resourceObject=/Game/Athena/UI/Frontend/Spatial/Art/T_Ui_Empty
     tiling=NoTile
     imageSize=<16 16>
     margin=<ltrb 0 0 0 0>
     outlineSettings {
      bUseBrushTransparency=false
      roundingType=HalfHeightRadius
      width=0
      color {colorUseRule=UseColor_Specified, specifiedColor=<rgba 0 0 0 0>}
      cornerRadii=<0 0 0 0>
     }
     tintColor {colorUseRule=UseColor_Specified, specifiedColor=<rgba 1 1 1 1>}
     uVRegion {bIsValid=false, max=<0 0>, min=<0 0>}
    }
   }
   slot {class=OverlaySlot, properties {horizontalAlignment=HAlign_Fill, verticalAlignment=VAlign_Fill}}
  }
  Image_Locked {
   widgetClassPath=&Image
   bInherited=true
   bIsVariable=true
   parent=Overlay_0
   widget=Image_Locked
   properties {
    visibility=SelfHitTestInvisible
    brush {
     bIsDynamicallyLoaded=false
     drawAs=Image
     imageType=NoImage
     mirroring=NoMirror
     resourceName=None
     resourceObject=/FortUILibrary/Styles/Border/Materials/M_CircledLock
     tiling=NoTile
     imageSize=<32 32>
     margin=<ltrb 0 0 0 0>
     outlineSettings {
      bUseBrushTransparency=false
      roundingType=HalfHeightRadius
      width=0
      color {colorUseRule=UseColor_Specified, specifiedColor=<rgba 0 0 0 0>}
      cornerRadii=<0 0 0 0>
     }
     tintColor {colorUseRule=UseColor_Specified, specifiedColor=<rgba 1 1 1 1>}
     uVRegion {bIsValid=false, max=<0 0>, min=<0 0>}
    }
   }
   slot {class=OverlaySlot, properties {horizontalAlignment=HAlign_Center, verticalAlignment=VAlign_Center}}
  }
  Image_Active {
   widgetClassPath=&Image
   bInherited=true
   bIsVariable=true
   parent=Overlay_0
   widget=Image_Active
   properties {
    visibility=SelfHitTestInvisible
    brush {
     bIsDynamicallyLoaded=false
     drawAs=Image
     imageType=NoImage
     mirroring=NoMirror
     resourceName=None
     resourceObject=/FortUILibrary/Systems/Items/MaterialInstances/MI_UIKit_EquippedStateOutline
     tiling=NoTile
     imageSize=<32 32>
     margin=<ltrb 0 0 0 0>
     outlineSettings {
      bUseBrushTransparency=false
      roundingType=HalfHeightRadius
      width=0
      color {colorUseRule=UseColor_Specified, specifiedColor=<rgba 0 0 0 0>}
      cornerRadii=<0 0 0 0>
     }
     tintColor {colorUseRule=UseColor_Specified, specifiedColor=<rgba 1 1 1 1>}
     uVRegion {bIsValid=false, max=<0 0>, min=<0 0>}
    }
   }
   slot {class=OverlaySlot, properties {horizontalAlignment=HAlign_Fill, verticalAlignment=VAlign_Fill}}
  }
  Overlay_Conflict {
   widgetClassPath=&Overlay
   bInherited=true
   bIsVariable=true
   parent=Overlay_0
   widget=Overlay_Conflict
   slot {class=OverlaySlot, properties {verticalAlignment=VAlign_Bottom, padding=<ltrb 8 0 0 8>}}
  }
  Image_Conflict {
   widgetClassPath=&Image
   parent=Overlay_Conflict
   widget=Image_Conflict
   properties {brush {
     bIsDynamicallyLoaded=false
     drawAs=Image
     imageType=NoImage
     mirroring=NoMirror
     resourceName=None
     resourceObject=/Game/Athena/UI/Frontend/CosmeticGating/Materials/MI_UI_CosmeticGate_WarningIcon
     tiling=NoTile
     imageSize=<28 28>
     margin=<ltrb 0 0 0 0>
     outlineSettings {
      bUseBrushTransparency=false
      roundingType=HalfHeightRadius
      width=0
      color {colorUseRule=UseColor_Specified, specifiedColor=<rgba 0 0 0 0>}
      cornerRadii=<0 0 0 0>
     }
     tintColor {colorUseRule=UseColor_Specified, specifiedColor=<rgba 1 1 1 1>}
     uVRegion {bIsValid=false, max=<0 0>, min=<0 0>}
    }}
   slot {class=OverlaySlot, properties {verticalAlignment=VAlign_Bottom}}
  }
  Block_Outline {
   widgetClassPath=/FortUILibrary/UIKit/Buttons/Blocks/WBP_UIKit_Block_Outline
   bIsVariable=true
   parent=Overlay_0
   widget=Block_Outline
   properties {
    bottomLeftCornerRadius=30
    bottomRightCornerRadius=30
    radiusInPixels=true
    topLeftCornerRadius=30
    topRightCornerRadius=30
    visibility=SelfHitTestInvisible
   }
   slot {class=OverlaySlot, properties {horizontalAlignment=HAlign_Fill, verticalAlignment=VAlign_Fill}}
  }
  InputActionWidget {widgetClassPath=&CommonActionWidget, bInherited=true, widget=null}
 ]
}
